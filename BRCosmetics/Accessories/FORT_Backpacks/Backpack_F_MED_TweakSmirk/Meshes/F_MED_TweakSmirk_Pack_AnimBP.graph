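# Animation Blueprint for the TweakSmirk backpack that offsets its pack bone to suit the wearer's body and pose.
# Each update derives an arm alpha from left-arm delta and melee state, plus backpack rotation and translation from spine yaw relative to reference pose.
# Customization loading records the body part's gender and size, feeding a not-large-male flag.
# The AnimGraph chains four additive bone modifiers from reference pose, using bool or float alphas.
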AnimBlueprint /BRCosmetics/Accessories/FORT_Backpacks/Backpack_F_MED_TweakSmirk/Meshes/F_MED_TweakSmirk_Pack_AnimBP
target_skeleton=&/F_MED_TweakSmirk_Pack_Skeleton
graphs [
 AnimGraph {nodes [
   n1 = "Transform (Modify) Bone - Bone: pack"!(ComponentPose=Local To Component(LocalPose=Local Space Ref Pose()), bAlphaBoolEnabled:bool, Translation="0,0.5,0") -> (Pose) {
    alphaInputType=Bool
    translationMode=BMM_Additive
    translationSpace=BCS_BoneSpace
    alphaBoolBlend {blendOption=Linear}
    boneToModify {boneName=pack}
    translation {y=0.5}
   }
   n2 = "Transform (Modify) Bone - Bone: pack"!(ComponentPose=@n1, bAlphaBoolEnabled, Translation="0,0,0", Rotation:Rotator="0,0,0") -> (Pose) {
    alphaInputType=Bool
    rotationMode=BMM_Additive
    translationMode=BMM_Additive
    alphaBoolBlend {blendOption=Linear}
    boneToModify {boneName=pack}
   }
   n3 = "Transform (Modify) Bone - Bone: pack"!(ComponentPose=@n2, Alpha:float, Translation="0,-3,0", Rotation="0,-35,0", Scale:Vector) -> (Pose) {
    rotationMode=BMM_Additive
    translationMode=BMM_Additive
    translationSpace=BCS_BoneSpace
    alphaScaleBiasClamp {bInterpResult=true}
    boneToModify {boneName=pack}
    rotation {yaw=-35}
    translation {y=-3}
   }
   n4 = "Transform (Modify) Bone - Bone: pack"!(ComponentPose=@n3, bAlphaBoolEnabled:bool, Translation="0,-1.5,0", Rotation="0,0,0", Scale) -> (Pose) {
    alphaInputType=Bool
    rotationMode=BMM_Additive
    translationMode=BMM_Additive
    translationSpace=BCS_BoneSpace
    alphaBoolBlend {blendInTime=0.25, blendOption=Linear, blendOutTime=0.25}
    boneToModify {boneName=pack}
    translation {y=-1.5}
   }
   Output Pose(Result=Component To Local(ComponentPose=@n4)) {layerGroup=None}
  ]}
 EventGraph {nodes [
   Event BlueprintUpdateAnimation()
   n1 = Get Skel_Mesh()
   > Set ArmAlpha(ArmAlpha:double=float * float(
     A=GetAlphaFromReferenceTransformDelta(SkeletalMeshComponent=@n1, BoneName=upperarm_l, BaseName=lowerarm_l, InputRange=(X=-40,Y=-140), TwistAxis="1,0,0")
     B:bool=Get bIsPlayingMeleeAnim()
    ) {is=title K2Node_PromotableOperator}) -> (Output_Get:double)
   n2 = Get Skel_Mesh()
   n3 = BreakRotator(InRot=BreakTransform(InTransform=GetDeltaTransformFromRefPose(self=@n2, BoneName=spine_05, BaseName=spine_01)).Rotation)
   > Set backpack_rotate(backpack_rotate=MakeRotator(Roll=MapRangeClamped(Value=@n3.Yaw, InRangeA=0, InRangeB=-20, OutRangeB=15)))
   > Set backpack_translation(backpack_translation=MakeVector(Y=MapRangeClamped(Value=@n3.Yaw, InRangeA=0, InRangeB=-20, OutRangeB=-2.5)))
   n4 = Get isLarge() -> (isLarge:bool)
   n5 = Get isMale() -> (isMale:bool)
   > Set is NOT M_Large(is NOT M_Large:bool=NOT Boolean(A=AND Boolean(A=@n5, B=@n4))) -> (Output_Get:bool)
   n6 = Event OnCharacterCustomizationLoaded()
   n7 = Get Array Item with EFort Custom Part Type(Array:Array of Custom Character Part Object References=@n6.CustomParts) -> (Output:Custom Character Part Object Reference)
   n8 = Get GenderPermitted!(self:Custom Character Part Object Reference=@n7) -> (GenderPermitted:EFortCustomGender Enum=Invalid)
   > Set isMale(isMale:bool="Equal (Enum)"(A:EFortCustomGender Enum=@n8, B:EFortCustomGender Enum=Male)) -> (Output_Get:bool)
   > Set isFemale(isFemale:bool="Equal (Enum)"(A=@n8, B=Female)) -> (Output_Get:bool)
   > Set isLarge(isLarge:bool="Equal (Enum)"(
     A:EFortCustomBodyType Enum=Get BodyTypesPermitted!(self:Custom Character Part Object Reference=@n7) -> (BodyTypesPermitted:EFortCustomBodyType Enum=NONE)
     B:EFortCustomBodyType Enum=Large
    )) -> (Output_Get:bool)
   Event BlueprintInitializeAnimation()
   > Set Skel_Mesh(Skel_Mesh=GetOwnerSkeletalMeshComponent())
  ]}
]
variables [
 ArmAlpha {type=number}
 backpack_rotate {type=Rotator}
 backpack_translation {type=Vector, default=<0 -1.4 0>}
 Body Mesh {type=SkeletalMeshComponent}
 is NOT M_Large {type=boolean, default=true}
 isFemale {type=boolean}
 isLarge {type=boolean}
 isMale {type=boolean}
 Owning Pawn {type=Pawn, default=null}
 Owning Player Pawn {type=FortPlayerPawn}
 Skel_Mesh {type=SkeletalMeshComponent}
]
